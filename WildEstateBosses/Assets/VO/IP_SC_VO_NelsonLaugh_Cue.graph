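# Sound cue for Nelson's laugh voice line in the WildEstateBosses plugin, classed as critical enemy VO.
# A Random node picks one of four equally weighted laugh waves, played with the NelsonLaugh attenuation at high priority and slightly boosted volume.
# Its concurrency override allows one instance per owner and blocks new plays while one is active.
# It also sends to a boss VO envelope-follower audio bus.
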SoundCue /WildEstateBosses/Assets/VO/IP_SC_VO_NelsonLaugh_Cue
root=Random_0
nodes [
 {id=Random_0, children [WavePlayer_0,WavePlayer_1,WavePlayer_2,WavePlayer_3], properties {weights [1,1,1,1]}}
 {id=WavePlayer_0, properties {soundWaveAssetPtr=/WildEstateBosses/Assets/VO/Waves/vo_IP_NelsonLaugh_013}}
 {id=WavePlayer_1, properties {soundWaveAssetPtr=/WildEstateBosses/Assets/VO/Waves/vo_IP_NelsonLaugh_012}}
 {id=WavePlayer_2, properties {soundWaveAssetPtr=/WildEstateBosses/Assets/VO/Waves/vo_IP_NelsonLaugh_011}}
 {id=WavePlayer_3, properties {soundWaveAssetPtr=/WildEstateBosses/Assets/VO/Waves/vo_IP_NelsonLaugh_010}}
]
properties {
 attenuationSettings=&/Gameplay_VO_NelsonLaugh_Att
 bOverrideConcurrency=true
 maxDistance=10000
 priority=100
 soundClassObject=&Fort_VO_Enemy_Critical
 volumeMultiplier=1.2
 busSends [{audioBus=/Game/Sounds/Modulation/EnvelopeFollowers/BR_Boss/AB_BR_Boss_VO_EnvFollow, minSendLevel=0, sendLevel=1, soundSourceBus=null}]
 concurrencyOverrides {bLimitToOwner=true, maxCount=1, resolutionRule=PreventNew}
}
